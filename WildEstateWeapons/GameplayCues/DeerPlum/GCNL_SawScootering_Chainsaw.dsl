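; Looping GameplayCue notify that plays audio and ground-spark effects while a player rides the chainsaw as a scooter.
; On start, it begins a sound indicator for non-locally-viewed pawns, then attaches Niagara sparks and surface audio, delayed unless the pawn has the falling-into-scooter tag.
; Weapon surface-change events look up a data table to update the audio's sound, volume and pitch parameters.
; Removal stops the indicator, deactivates sparks, unbinds events and fades audio.
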(event EventOnLoopingStart(EitherVFX-preferthisversion)
  (params (MyTarget :type Actor%20Object%20Reference) (Parameters :type Gameplay%20Cue%20Parameters%20Structure%20%28by%20ref%29) (ParticleComponents :type Array%20of%20FXSystem%20Component%20Object%20References) (AudioComponents :type Array%20of%20Audio%20Component%20Object%20References))
  (|CallParentFunction MyTarget Parameters ParticleComponents AudioComponents)
  (bind _asfort_player_pawn (Utilities|Casting|CastToFortPlayerPawn MyTarget))
  (bind _asb_weapon_chainsaw (Utilities|Casting|CastToB_Weapon_Chainsaw (Weapon|GetCurrentWeapon _asfort_player_pawn)))
  (Variables|Default|SetChainsawWeapon _asb_weapon_chainsaw)
  (CallFunction|StartLoopingSound _asfort_player_pawn)
  (if (not (GameplayTags|HasMatchingGameplayTag _asfort_player_pawn "(TagName=\"Granted.Athena.Chainsaw.FallingIntoScooter\")"))
    (Utilities|FlowControl|Delay self 0.65))
  (Default|BindEventtoOnScooterSurfaceChanged (Variables|Default|GetChainsawWeapon) (EventDispatchers|CreateEvent :!event Custom|OnSurfaceUpdated))
  (CallFunction|SpawnAudioComponent)
  (bind _returnvalue (Niagara|SpawnSystemAttached "/WildEstateWeapons/Assets/FX/DeerPlum/NS_DeerPlum_CS_GroundSparks.NS_DeerPlum_CS_GroundSparks" (Variables|/Script/FortniteGame.FortWeapon|GetWeaponMesh (Variables|Default|GetChainsawWeapon)) "root" "0, 0, 0" "0, 0, 0" "SnapToTarget" true true "AutoRelease"))
  (Variables|Transient|SetGroundSparksNiagaraSystem _returnvalue))

(event EventOnRemoval(EitherVFX-preferthisversion)
  (params (MyTarget :type Actor%20Object%20Reference) (Parameters :type Gameplay%20Cue%20Parameters%20Structure%20%28by%20ref%29) (ParticleComponents :type Array%20of%20FXSystem%20Component%20Object%20References) (AudioComponents :type Array%20of%20Audio%20Component%20Object%20References) (BurstCameraShakeInstance :type Camera%20Shake%20Base%20Object%20Reference) (BurstDecalInstance :type Decal%20Actor%20Object%20Reference))
  (|CallParentFunction MyTarget Parameters ParticleComponents AudioComponents BurstCameraShakeInstance BurstDecalInstance)
  (SoundIndicator|StopSoundIndicatorLooping (Variables|Default|GetSoundIndicator_ScooterLoop))
  (bind _groundsparksniagarasystem (Variables|Transient|GetGroundSparksNiagaraSystem)
    (:then
      (Components|Activation|Deactivate _groundsparksniagarasystem))
    (:else
      _))
  (bind _chainsawweapon (Variables|Default|GetChainsawWeapon)
    (:then
      (Default|UnbindallEventsfromOnScooterSurfaceChanged _chainsawweapon))
    (:else
      _))
  (bind _surfaceaudiocomponent (Variables|Default|GetSurfaceAudioComponent)
    (:then
      (Audio|Components|Audio|FadeOut _surfaceaudiocomponent 0.3 0.0 "Linear" :!class /Script/Engine.AudioComponent))
    (:else)))

(event Custom|SpawnAudioComponent
  (bind _chainsawweapon (Variables|Default|GetChainsawWeapon)
    (:then
      (bind _returnvalue (Audio|SpawnSoundAttached "/ChainsawGameplay/Assets/Sounds/MSS_Chainsaw_Surface_ScooterLoop.MSS_Chainsaw_Surface_ScooterLoop" (Variables|/Script/Engine.Actor|GetRootComponent _chainsawweapon)))
      (Variables|Default|SetSurfaceAudioComponent _returnvalue)
      (CallFunction|OnSurfaceUpdated (Variables|/ChainsawGameplay/Gameplay/B_Weapon_Chainsaw.B_Weapon_Chainsaw_C|GetCurrentSurface _chainsawweapon)))
    (:else)))

(event Custom|OnSurfaceUpdated
  (params (NewSurface :type EPhysicalSurface%20Enum :default "SurfaceType_Default"))
  (bind (_surfacesounddata_surfacesound_8_4fe7bd224137f616108d6d9a28013055 _surfacesounddata_volumemultiplier_10_370561ff448c62ed90c9a0abaaad3e40 _surfacesounddata_pitchshift_9_2e667f4b45c83f3fbb015e9c0978b909) (Utilities|Struct|BreakSurfaceSoundStruct (Variables|GetSurfaceSoundData_SurfaceSound_8_4FE7BD224137F616108D6D9A28013055)))
  ;[1 1 1] Surface change logic
  (bind _returnvalue (Utilities|GetDataTableRow "/ChainsawGameplay/Assets/Sounds/DT_ChainsawSurfaceSounds.DT_ChainsawSurfaceSounds" (Utilities|String|StringToName (Utilities|String|EnumtoString NewSurface)))
    (:then
      (Variables|Default|SetSurfaceSoundData _returnvalue))
    (:RowNotFound
      (Variables|Default|SetSurfaceSoundData (Utilities|Struct|MakeSurfaceSoundStruct (Utilities|Array|MakeArray 0)))))
  (bind _surfaceaudiocomponent (Variables|Default|GetSurfaceAudioComponent)
    (:then
      (Audio|Parameter|SetObjectParameter _surfaceaudiocomponent "Surfaces.Sound" (select (Utilities|Array|IsNotEmpty _surfacesounddata_surfacesound_8_4fe7bd224137f616108d6d9a28013055) (Utilities|Array|RandomArrayItem _surfacesounddata_surfacesound_8_4fe7bd224137f616108d6d9a28013055) 0))
      (Audio|Parameter|SetFloatParameter _surfaceaudiocomponent "Surfaces.Volume" _surfacesounddata_volumemultiplier_10_370561ff448c62ed90c9a0abaaad3e40 :!class /Script/AudioExtensions.AudioParameterControllerInterface)
      (Audio|Parameter|SetFloatParameter _surfaceaudiocomponent "Surfaces.Pitch" _surfacesounddata_pitchshift_9_2e667f4b45c83f3fbb015e9c0978b909 :!class /Script/AudioExtensions.AudioParameterControllerInterface)
      (Audio|Parameter|ExecuteTriggerParameter _surfaceaudiocomponent "Surfaces.OnSurfaceChanged"))
    (:else)))

(fn StartLoopingSound
  (params (TargetedActor :type Fort%20Player%20Pawn%20Object%20Reference))
  (if (not (Pawn|IsLocallyViewed TargetedActor))
    (SoundIndicator|StartTargetedSoundIndicatorLooping (Variables|Default|GetSoundIndicator_ScooterLoop) TargetedActor "/ChainsawGameplay/Assets/Sounds/ChainSaw_Scooter_Loop_Cue.ChainSaw_Scooter_Loop_Cue" 1.0 0)))

(fn ConstructionScript)
