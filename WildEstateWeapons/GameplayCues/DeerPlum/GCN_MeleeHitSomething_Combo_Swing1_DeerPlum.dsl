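; GameplayCue notify blueprint for the first swing hit of the DeerPlum melee combo, apparently still a stub.
; On BeginPlay it calls the parent implementation and prints a debug "CUE FIRED" string.
; The construction script is empty and the graph itself spawns no effects or sounds.
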(event EventBeginPlay
  (|CallParentFunction)
  (Development|PrintString 0 "CUE FIRED"))

(fn ConstructionScript)
